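# Niagara trail effect for the SquidGlisten LiftDairy glider, combining skeletal-mesh ribbon trails with sprite speedlines.
# Two AnimTrail emitters draw socket-based ribbons with variant-selected colors, one also picking sockets by VariantID, faded by age, time of day and a system integrator.
# Speedlines spawn on a torus, stretch with speed, lean with Right/ForwardDot, and fade via a system Timeline.
# Front-end modifiers substitute motion and opacity for menu previews.
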NiagaraSystem /SalineCabinet/Assets/SalineCabinet_Grape/FX/NS_Glider_SquidGlisten_LiftDairy_Trail
dependencies {
 usedDataInterfaces [
  &NiagaraDataInterfaceArrayInt32
  &NiagaraDataInterfaceCurve
  &NiagaraDataInterfaceColorCurve
  &NiagaraDataInterfaceVector2DCurve
  &NiagaraDataInterfaceSkeletalMesh
 ]
 usedDynamicInputs [&Boolean_AndOperation, &SetBoolByFloatComparison]
 usedModules [
  /Game/Weapons/FORT_Melee/Pickaxe_GateHound/FX/VariantManager/FN_VariantManager
  &self:SystemSpawnScript.SystemScriptSource.SystemScriptGraph.NiagaraNodeAssignment_0.SetVariables_#1
  &SystemState
  &FN_Integrator
  &Timeline
 ]
}
emitters {
 FN_AnimTrail {
  emitterSpawnScript [&FN_RibbonIDOffsetUtility()]
  emitterUpdateScript [
   &EmitterState(Life Cycle Mode=::NewEnumerator1, Loop Duration=1, UseLoopDelay=false)
   &FN_SpawnAnimTrail(Number of Anim Trails=2)
  ]
  particleSpawnScript [
   &PartitionParticles()
   &FN_InitializeAnimTrailBoneIndecies(
    First Bone or Socket Index:NiagaraInt32=&SelectIntFromArray(
     Direct Array Index=&MakeIntFromFloat(Float=$User.VariantID)
     Int Selection Array=DataInterface() {propertyValues {IntData [1,6]}}
    )
    Second Bone or Socket Index=0
   )
   &FN_InitializeAnimTrailBoneIndecies(
    AnimTrail Selection=1
    First Bone or Socket Index:NiagaraInt32=&SelectIntFromArray(
     Direct Array Index=&MakeIntFromFloat(Float=$User.VariantID)
     Int Selection Array=DataInterface() {propertyValues {IntData [3,7]}}
    )
    Second Bone or Socket Index=2
   ) {module=FN_InitializeAnimTrailBoneIndecies001}
   &FN_InitializeAnimTrail(
    Color=/Game/Weapons/FORT_Melee/Pickaxe_GateHound/FX/VariantManager/VariantColorSelector(
     Variant Amount:NiagaraInt32=0
     BlendAmount:NiagaraFloat=?
     Morning Color:LinearColor=?
     Day Color:LinearColor=?
     Evening Color:LinearColor=?
     Night Color:LinearColor=?
     Use TOD?:NiagaraBool=false
     Use TODCurve?:NiagaraBool=false
     Original Color:LinearColor=<rgba 0.0001 0.01 0.0005 1>
     Variant 1 Color:LinearColor=<rgba 1 0.0134 0.0213 1>
     Variant 2 Color:LinearColor=?
     Variant 3 Color:LinearColor=?
     Variant 4 Color:LinearColor=?
     Variant 5 Color:LinearColor=?
     Use FrontEnd?:NiagaraBool=false
     Front End Color:LinearColor=?
     Gameplay Color:LinearColor=?
    )
    Lifetime=0.25
    SkeletalMesh=$System.Skeletal Mesh
   )
   &FN_InitializeAnimTrail(
    Color=&VariantColorSelector(
     Variant Amount:NiagaraInt32=0
     BlendAmount:NiagaraFloat=?
     Morning Color:LinearColor=?
     Day Color:LinearColor=?
     Evening Color:LinearColor=?
     Night Color:LinearColor=?
     Use TOD?:NiagaraBool=false
     Use TODCurve?:NiagaraBool=false
     Original Color:LinearColor=<rgba 1.5 0.0202 0.0319 1>
     Variant 1 Color:LinearColor=<rgba 0.0001 0.01 0.0005 1>
     Variant 2 Color:LinearColor=?
     Variant 3 Color:LinearColor=?
     Variant 4 Color:LinearColor=?
     Variant 5 Color:LinearColor=?
     Use FrontEnd?:NiagaraBool=false
     Front End Color:LinearColor=?
     Gameplay Color:LinearColor=?
    )
    Lifetime=0.25
    Position Offset=<0 0 -2.5>
    SkeletalMesh=$System.Skeletal Mesh
   ) {enabled=false, module=FN_InitializeAnimTrail001}
   &AddVelocity(Velocity=&FN_FrontEndVector3Modifier(Gameplay=<0 0 0>, Front End=<0 -1000 0>))
  ]
  particleUpdateScript [
   &SolveForcesAndVelocity(Clamp Velocity=false, Limit Acceleration=false)
   &Color(Color=&MakeLinearColorFromVectorAndFloat(
     "Vector (RGB)"=<1 1 1>
     "Float (Alpha)"=&Multiply_Float(
      A=&FloatFromCurve(
       FloatCurve=DataInterface() {propertyValues {Curve {keys [
           {arriveTangent=-1.5186, interpMode=RCIM_Cubic, leaveTangent=-1.5186, tangentMode=RCTM_User, time=0, value=0.8}
           {arriveTangent=-0.6904, interpMode=RCIM_Cubic, leaveTangent=-0.6904, tangentMode=RCTM_User, time=0.3, value=0.25}
           {interpMode=RCIM_Cubic, tangentMode=RCTM_User, time=1, value=0}
          ]}}}
       Scale Curve=2
      )
      B=&FN_FrontEndFloatModifier(
       Gameplay=1
       Front End=&FloatFromCurve(FloatCurve=DataInterface() {propertyValues {Curve {keys [<at 1 0>,<at 3 1>]}}}, CurveIndex=$Emitter.Age)
      )
     )
    )) {enabled=false}
   &FN_UpdateAnimTrail(Locked Particle RibbonWidth=1, SkeletalMesh=$System.Skeletal Mesh) {module=FN_UpdateAnimTrail001}
   /SetVariables(Particles.RibbonU0Override=$Particles.NormalizedAge) {isSetParameters=true, module=SetVariables_#2}
   &ScaleColor(
    Scale Alpha=&Multiply_Float(
     A=&FloatFromCurve(
      FloatCurve=DataInterface() {propertyValues {Curve {keys [
          {arriveTangent=-1.5186, interpMode=RCIM_Cubic, leaveTangent=-1.5186, tangentMode=RCTM_User, time=0, value=0.8}
          {arriveTangent=-0.6904, interpMode=RCIM_Cubic, leaveTangent=-0.6904, tangentMode=RCTM_User, time=0.3, value=0.25}
          {interpMode=RCIM_Cubic, tangentMode=RCTM_User, time=1, value=0}
         ]}}}
      Scale Curve=2
     )
     B=&FN_FrontEndFloatModifier(
      Gameplay=1
      Front End=&FloatFromCurve(FloatCurve=DataInterface() {propertyValues {Curve {keys [<at 1 0>,<at 3 1>]}}}, CurveIndex=$Emitter.Age)
     )
    )
    ScaleA=true
    ScaleRGB=true
   ) {module=ScaleColor001}
   &ScaleColor(
    Color Value To Scale=$Particles.Color
    Scale Alpha=&Multiply_Float(
     A=&FN_FrontEndFloatModifier(Gameplay=&Multiply_Float(A=$System.FN_Integrator.Value, B=0.65), Front End=1)
     B=&FNTimeOfDayScalar(Night Scale=0.1, Morning Scale=0.3, Day Scale=0.3, Evening Scale=0.1)
    )
    ScaleA=true
    ScaleRGB=true
   )
   &DynamicMaterialParameters(Module.Index0Float0Write=true, Module.Index0Float1Write=true, Module.Index0Float2Write=true, Module.Index0Float3Write=true)
  ]
  rendererClasses [&NiagaraRibbonRendererProperties]
  renderers [{rendererClass=&NiagaraRibbonRendererProperties}]
 }
 FN_AnimTrail001 {
  emitterSpawnScript [&FN_RibbonIDOffsetUtility()]
  emitterUpdateScript [
   &EmitterState(Life Cycle Mode=::NewEnumerator1, Loop Duration=1, UseLoopDelay=false)
   &FN_SpawnAnimTrail(Number of Anim Trails=4)
  ]
  particleSpawnScript [
   &FN_InitializeAnimTrailBoneIndecies(First Bone or Socket Index=1, Second Bone or Socket Index=4)
   &FN_InitializeAnimTrailBoneIndecies(AnimTrail Selection=1, First Bone or Socket Index=4) {module=FN_InitializeAnimTrailBoneIndecies002}
   &FN_InitializeAnimTrailBoneIndecies(AnimTrail Selection=2, First Bone or Socket Index=3, Second Bone or Socket Index=5) {module=FN_InitializeAnimTrailBoneIndecies001}
   &FN_InitializeAnimTrailBoneIndecies(AnimTrail Selection=3, First Bone or Socket Index=5, Second Bone or Socket Index=3) {module=FN_InitializeAnimTrailBoneIndecies003}
   &FN_InitializeAnimTrail(
    AnimTrailWidth=1.5
    Color=&VariantColorSelector(
     Variant Amount:NiagaraInt32=0
     BlendAmount:NiagaraFloat=?
     Morning Color:LinearColor=?
     Day Color:LinearColor=?
     Evening Color:LinearColor=?
     Night Color:LinearColor=?
     Use TOD?:NiagaraBool=false
     Use TODCurve?:NiagaraBool=false
     Original Color:LinearColor=<rgba 0.6562 0.3954 0.0016 1>
     Variant 1 Color:LinearColor=<rgba 1.5 0.0202 0.0319 0>
     Variant 2 Color:LinearColor=?
     Variant 3 Color:LinearColor=?
     Variant 4 Color:LinearColor=?
     Variant 5 Color:LinearColor=?
     Use FrontEnd?:NiagaraBool=false
     Front End Color:LinearColor=?
     Gameplay Color:LinearColor=?
    )
    Lifetime=0.25
    SkeletalMesh=$System.Skeletal Mesh
   )
   &FN_InitializeAnimTrail(
    Color=&VariantColorSelector(
     Variant Amount:NiagaraInt32=0
     BlendAmount:NiagaraFloat=?
     Morning Color:LinearColor=?
     Day Color:LinearColor=?
     Evening Color:LinearColor=?
     Night Color:LinearColor=?
     Use TOD?:NiagaraBool=false
     Use TODCurve?:NiagaraBool=false
     Original Color:LinearColor=<rgba 1.5 0.0202 0.0319 1>
     Variant 1 Color:LinearColor=<rgba 0.0001 0.01 0.0005 1>
     Variant 2 Color:LinearColor=?
     Variant 3 Color:LinearColor=?
     Variant 4 Color:LinearColor=?
     Variant 5 Color:LinearColor=?
     Use FrontEnd?:NiagaraBool=false
     Front End Color:LinearColor=?
     Gameplay Color:LinearColor=?
    )
    Lifetime=0.25
    Position Offset=<0 0 -2.5>
    SkeletalMesh=$System.Skeletal Mesh
   ) {enabled=false, module=FN_InitializeAnimTrail001}
   &AddVelocity(Velocity=&FN_FrontEndVector3Modifier(Gameplay=<0 0 0>, Front End=<0 -1000 0>))
  ]
  particleUpdateScript [
   &SolveForcesAndVelocity(Clamp Velocity=false, Limit Acceleration=false)
   &Color(Color=&MakeLinearColorFromVectorAndFloat(
     "Vector (RGB)"=<1 1 1>
     "Float (Alpha)"=&Multiply_Float(
      A=&FloatFromCurve(
       FloatCurve=DataInterface() {propertyValues {Curve {keys [
           {arriveTangent=-1.5186, interpMode=RCIM_Cubic, leaveTangent=-1.5186, tangentMode=RCTM_User, time=0, value=0.8}
           {arriveTangent=-0.6904, interpMode=RCIM_Cubic, leaveTangent=-0.6904, tangentMode=RCTM_User, time=0.3, value=0.25}
           {interpMode=RCIM_Cubic, tangentMode=RCTM_User, time=1, value=0}
          ]}}}
       Scale Curve=2
      )
      B=&FN_FrontEndFloatModifier(
       Gameplay=1
       Front End=&FloatFromCurve(FloatCurve=DataInterface() {propertyValues {Curve {keys [<at 1 0>,<at 3 1>]}}}, CurveIndex=$Emitter.Age)
      )
     )
    )) {enabled=false}
   &FN_UpdateAnimTrail(Locked Particle RibbonWidth=1, SkeletalMesh=$System.Skeletal Mesh) {module=FN_UpdateAnimTrail001}
   /SetVariables(Particles.RibbonU0Override=$Particles.NormalizedAge) {isSetParameters=true, module=SetVariables_#2}
   &ScaleColor(
    Scale Alpha=&Multiply_Float(
     A=&FloatFromCurve(
      FloatCurve=DataInterface() {propertyValues {Curve {keys [
          {arriveTangent=-1.5186, interpMode=RCIM_Cubic, leaveTangent=-1.5186, tangentMode=RCTM_User, time=0, value=0.8}
          {arriveTangent=-0.6904, interpMode=RCIM_Cubic, leaveTangent=-0.6904, tangentMode=RCTM_User, time=0.3, value=0.25}
          {interpMode=RCIM_Cubic, tangentMode=RCTM_User, time=1, value=0}
         ]}}}
      Scale Curve=2
     )
     B=&FN_FrontEndFloatModifier(
      Gameplay=1
      Front End=&FloatFromCurve(FloatCurve=DataInterface() {propertyValues {Curve {keys [<at 1 0>,<at 3 1>]}}}, CurveIndex=$Emitter.Age)
     )
    )
    ScaleA=true
    ScaleRGB=true
   ) {module=ScaleColor001}
   &ScaleColor(
    Color Value To Scale=$Particles.Color
    Scale Alpha=&Multiply_Float(
     A=&FN_FrontEndFloatModifier(Gameplay=&Multiply_Float(A=$System.FN_Integrator.Value, B=0.5), Front End=1)
     B=&FNTimeOfDayScalar(Night Scale=0.1, Morning Scale=0.3, Day Scale=0.3, Evening Scale=0.1)
    )
    ScaleA=true
    ScaleRGB=true
   )
  ]
  rendererClasses [&NiagaraRibbonRendererProperties]
  renderers [{rendererClass=&NiagaraRibbonRendererProperties}]
 }
 Speedlines001 {
  emitterUpdateScript [
   &EmitterState()
   &SpawnRate(
    SpawnRate=&MakeCustomFloatFromBool(Bool=$User.bIsFrontEndPreview, True Float=50, False Float=90)
    Use Spawn Probability=true
   ) {module=SpawnRate001}
  ]
  particleSpawnScript [
   &InitializeParticle(
    Color=&MakeLinearColorFromVectorAndFloat(
     "Vector (RGB)"=<1 1 1>
     "Float (Alpha)"=&Multiply_Float(
      A=&UniformRangedFloat(Minimum=1, Maximum=1)
      B=&MakeCustomFloatFromBool(
       Bool=$User.bIsFrontEndPreview
       True Float=&FloatFromCurve(
        FloatCurve=DataInterface() {propertyValues {Curve {keys [{tangentMode=RCTM_User, time=0, value=0}, {tangentMode=RCTM_User, time=2, value=1}]}}}
        CurveIndex=$Emitter.Age
        Scale Curve=0.4
       )
       False Float=1
      )
     )
    )
    Lifetime Max=0.2
    Lifetime Min=0.15
    Lifetime Mode=::NewEnumerator1
    Position Offset=&FN_FrontEndVector3Modifier(Gameplay=<0 -50 0>, Front End=<0 50 0>)
    Sprite Size Max=<3 10>
    Sprite Size Min=<2 2>
    Sprite Size Mode=::NewEnumerator2
    UsePositionOffset=true
    Write Material Random=true
   )
   &AddVelocity(
    Module.Use Falloff Distance=false
    Override Seed=false
    Velocity Mode=::NewEnumerator1
    Velocity Speed=&FN_FrontEndFloatModifier(Gameplay=2500, Front End=-2500)
   ) {module=AddVelocity001}
   &TorusLocation(
    Align Torus With Axis=true
    Coordinate Space:ENiagaraCoordinateSpace=::Simulation
    Handle Radius=0
    Large Radius:NiagaraFloat=&FN_FrontEndFloatModifier(
     Gameplay=&RandomRangeFloat(Minimum=450, Maximum=500)
     Front End=&RandomRangeFloat(Minimum=250, Maximum=300)
    )
    MaskSpawnGroup=false
    Module.Override Randomness=false
    Offset=<0 33 0>
    Torus Axis=<0 1 0>
    Torus Origin:NiagaraPosition=&AddVectorToPosition(
     Position=$Particles.Position
     Vector=&FN_FrontEndVector3Modifier(Gameplay=<0 500 0>, Front End=<0 0 0>)
    )
   )
   &AddVelocity(
    Velocity=&Add_Vector(
     A=&RandomRangeVector(Minimum=<0 300 0>, Maximum=<0 700 1>)
     B=&Add_Vector(
      A=&Multiply_VectorByFloat(Vector=<-555 0 0>, Float=$User.RightDot)
      B=&Multiply_VectorByFloat(Vector=<0 0 222>, Float=$User.ForwardDot)
     )
    )
    Velocity Speed Scale=&FN_FrontEndFloatModifier(Gameplay=1, Front End=0.3)
   )
  ]
  particleUpdateScript [
   &SolveForcesAndVelocity(Clamp Velocity=false, Limit Acceleration=false)
   &ParticleState()
   &ScaleSpriteSize(Scale Factor=&Vector2DFromCurve(
     Vector2Curve=DataInterface() {propertyValues {XCurve {keys [<at 0 0.5 RCIM_Cubic>,<at 0.5 1 RCIM_Cubic>]}, YCurve {keys [<at 0 0 RCIM_Cubic>,<at 0.5 1 RCIM_Cubic>]}}}
     ScaleCurve=&Vector2DFromFloat(Value=1)
    ))
   &ScaleSpriteSizeBySpeed(Max Scale Factor=<1 50>, Min Scale Factor=<0.01 0.05>, Velocity Threshold=1)
   &ScaleColor(
    Scale Alpha=&FloatFromCurve(
     FloatCurve=DataInterface() {propertyValues {Curve {keys [
         <at 0 0 RCIM_Cubic>
         <at 0.3759 1.0207>
         {arriveTangent=-2.221, interpMode=RCIM_Cubic, leaveTangent=-2.221, tangentMode=RCTM_User, time=1, value=0}
        ]}}}
     Scale Curve=&MakeCustomFloatFromBool(
      Bool=$User.bIsFrontEndPreview
      True Float=&MakeCustomFloatFromBool(Bool=$User.bIsFullyDeployed, True Float=0.5, False Float=0)
      False Float=&Lerp_Float(A=0, B=0.5, Alpha=&FN_FrontEndFloatModifier(Gameplay=$System.Timeline.Time, Front End=0.5))
     )
    )
    Scale RGB=&VectorFromFloat(Value=&FN_FrontEndFloatModifier(Gameplay=0.5, Front End=0.33))
    ScaleA=true
    ScaleRGB=true
   )
  ]
  rendererClasses [&NiagaraSpriteRendererProperties]
  renderers [{rendererClass=&NiagaraSpriteRendererProperties}]
 }
}
summary {
 systemName=NS_Glider_SquidGlisten_LiftDairy_Trail
 emitters [
  {emitterName=FN_AnimTrail, rendererClasses [&NiagaraRibbonRendererProperties]}
  {emitterName=FN_AnimTrail001, rendererClasses [&NiagaraRibbonRendererProperties]}
  {emitterName=Speedlines001, rendererClasses [&NiagaraSpriteRendererProperties]}
 ]
 userVariables [
  User.bIsFrontEnd {type=NiagaraBool}
  User.bIsFrontEndPreview {type=NiagaraBool}
  User.bIsFullyDeployed {type=NiagaraBool}
  User.ForwardDot {type=NiagaraFloat, defaultValue=0, description=""}
  User.IsFrontEnd {type=NiagaraBool}
  User.PlayerSpeed {type=NiagaraFloat, defaultValue=0, description=""}
  User.RibbonIDOffset {type=NiagaraInt32}
  User.RightDot {type=NiagaraFloat, defaultValue=0, description=""}
  User.VariantID {type=NiagaraFloat, defaultValue=0, description=""}
 ]
}
